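# Post-process material for the Impostor Baker that outputs selected scene buffers for impostor capture.
# MPC_ImposterCapture weights mask and sum stored base color, stored specular, roughness, subsurface color, remapped world normal and depth, while EnableCapture fades out the normal scene image.
# Depth is computed by transforming world position into the capture's Origin and X/Y/Z basis, normalizing by R, remapping and inverting the Z component.
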Material /Game/ImpostorBaker/Materials/PostProcess/M_Capture_PostProcess
graph {
 properties {materialDomain=MD_PostProcess, pixelDepthOffsetMode=PDOM_Legacy, usedShadingModels=Default Lit}
 nodes [
  n1 = CollectionParameter() {collection=/Game/ImpostorBaker/MPC/MPC_ImposterCapture, parameterName=EnableCapture}
  n2 = SceneTexture() {sceneTextureId=PPI_PostProcessInput0}
  n3 = Multiply(
   A=CollectionParameter() {collection=&MPC_ImposterCapture, parameterName=BaseColor}
   B=ComponentMask(None=SceneTexture() {sceneTextureId=PPI_StoredBaseColor}.Color) {b=true, g=true, r=true}
  )
  n4 = Multiply(
   A=CollectionParameter() {collection=&MPC_ImposterCapture, parameterName=Specular}
   B=ComponentMask(None=SceneTexture() {sceneTextureId=PPI_StoredSpecular}.Color) {b=true, g=true, r=true}
  )
  n5 = CollectionParameter() {collection=&MPC_ImposterCapture, parameterName=Roughness}
  n6 = ComponentMask(None=SceneTexture() {sceneTextureId=PPI_Roughness}.Color) {r=true}
  n7 = SceneTexture() {sceneTextureId=PPI_SubsurfaceColor}
  n8 = Add(
   A=Add(A=Add(A=@n3, B=@n4), B=Multiply(A=@n5, B=@n6))
   B=Multiply(
    A=CollectionParameter() {collection=&MPC_ImposterCapture, parameterName=SubSurface}
    B=ComponentMask(None=@n7.Color) {b=true, g=true, r=true}
   )
  )
  n9 = SceneTexture() {sceneTextureId=PPI_WorldNormal}
  n10 = Multiply(
   A=CollectionParameter() {collection=&MPC_ImposterCapture, parameterName=Normal}
   B=ConstantBiasScale(None=ComponentMask(None=@n9.Color) {b=true, g=true, r=true})
  )
  n11 = Subtract(
   A=WorldPosition().XYZ
   B=ComponentMask(None=CollectionParameter() {collection=&MPC_ImposterCapture, parameterName=Origin}) {b=true, g=true, r=true}
  )
  n12 = ComponentMask(None=CollectionParameter() {collection=&MPC_ImposterCapture, parameterName=X}) {b=true, g=true, r=true}
  n13 = ComponentMask(None=CollectionParameter() {collection=&MPC_ImposterCapture, parameterName=Y}) {b=true, g=true, r=true}
  n14 = ComponentMask(None=CollectionParameter() {collection=&MPC_ImposterCapture, parameterName=Z}) {b=true, g=true, r=true}
  n15 = ConstantBiasScale(None=Divide(
    A=MaterialFunctionCall!(VectorToTransform=@n11, BasisX=@n12, BasisY=@n13, BasisZ=@n14) {materialFunction=/Engine/Functions/Engine_MaterialFunctions02/Math/InverseTransformMatrix}.Transformed Vector 3
    B=CollectionParameter() {collection=&MPC_ImposterCapture, parameterName=R}
   ))
  n16 = OneMinus(None=Saturate(None=ComponentMask(None=@n15) {b=true}))
  n17 = CollectionParameter() {collection=&MPC_ImposterCapture, parameterName=Depth}
  Add(
   A=Multiply(A=OneMinus(None=@n1), B=ComponentMask(None=@n2.Color) {b=true, g=true, r=true})
   B=Add(A=Add(A=@n8, B=@n10), B=Multiply(A=@n16, B=@n17))
  )
  ComponentMask(None=SceneTexture() {sceneTextureId=PPI_WorldNormal}.Color) {b=true, g=true, r=true}
  ComponentMask(None=SceneTexture() {sceneTextureId=PPI_BaseColor}.Color) {b=true, g=true, r=true}
  Constant3Vector() {constant=<rgba 0 1 0 0>}
 ]
}
